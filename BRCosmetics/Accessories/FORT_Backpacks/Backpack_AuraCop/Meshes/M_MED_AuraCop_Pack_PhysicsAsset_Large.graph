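# Large physics asset for the AuraCop backpack cosmetic, defining collision and joints for its dangling ropes and zipper.
# Three capsules form the pack body, with small base bodies attached by angularly free constraints as roots for each chain.
# Left and right five-segment rope chains use twist-locked, swing-limited joints, differing slightly per side.
# Short top ropes have tight swing and twist limits, while the zipper twists freely.
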PhysicsAsset /BRCosmetics/Accessories/FORT_Backpacks/Backpack_AuraCop/Meshes/M_MED_AuraCop_Pack_PhysicsAsset_Large
bodies {
 base_rope_l [{shapeType=Capsule, length=0.505, radius=0.505, shapeName=base_rope_l_capsule, rotation=<rot 0 -90 0>}]
 base_rope_r [{shapeType=Capsule, length=0.505, radius=0.505, shapeName=base_rope_r_capsule, rotation=<rot 0 -90 0>}]
 base_top_rope_l [{shapeType=Capsule, length=0.505, radius=0.505, shapeName=base_top_rope_l_capsule, rotation=<rot 0 -90 0>}]
 base_top_rope_r [{shapeType=Capsule, length=0.505, radius=0.505, shapeName=base_top_rope_r_capsule, rotation=<rot 0 -90 0>}]
 base_zipper [{shapeType=Capsule, length=0, radius=0.5, shapeName=base_zipper_capsule, rotation=<rot 0 -90 0>}]
 dyn_rope_00_l [{
   shapeType=Capsule
   length=4.4172
   radius=1.3699
   shapeName=dyn_rope_00_r_capsule
   center=<2.8271 0.0361 -0.01>
   rotation=<rot 0 -90.0242 -89.8655>
  }]
 dyn_rope_00_r [{
   shapeType=Capsule
   length=4.4172
   radius=1.3699
   shapeName=dyn_rope_00_r_capsule
   center=<2.8271 0.0361 -0.01>
   rotation=<rot 0 -90.0242 -89.8655>
  }]
 dyn_rope_01_l [{
   shapeType=Capsule
   length=4.7871
   radius=1.4
   shapeName=dyn_rope_01_r_capsule
   center=<1.7546 0.0366 -0.0075>
   rotation=<rot 0 -90.0242 -89.8655>
  }]
 dyn_rope_01_r [{
   shapeType=Capsule
   length=4.7871
   radius=1.4
   shapeName=dyn_rope_01_r_capsule
   center=<1.7546 0.0366 -0.0075>
   rotation=<rot 0 -90.0242 -89.8655>
  }]
 dyn_rope_02_l [{shapeType=Capsule, length=6.005, radius=1.4, shapeName=dyn_rope_02_r_capsule, center=<2.5391 -0.0089 -0.0065>, rotation=<rot 90 0 0>}]
 dyn_rope_02_r [{shapeType=Capsule, length=6.005, radius=1.4, shapeName=dyn_rope_02_r_capsule, center=<2.5391 -0.0089 -0.0065>, rotation=<rot 90 0 0>}]
 dyn_rope_03_l [{shapeType=Capsule, length=2.505, radius=1.2, shapeName=dyn_rope_03_r_capsule, center=<2.2079 -0.0559 0.0039>, rotation=<rot 90 0 0>}]
 dyn_rope_03_r [{shapeType=Capsule, length=2.505, radius=1.2, shapeName=dyn_rope_03_r_capsule, center=<2.2079 -0.0559 0.0039>, rotation=<rot 90 0 0>}]
 dyn_rope_04_l [{
   shapeType=Capsule
   length=12.6372
   radius=2.6231
   shapeName=dyn_rope_04_r_capsule
   center=<8.9681 0.2863 -0.0566>
   rotation=<rot 0 -87.6927 -89.7994>
  }]
 dyn_rope_04_r [{
   shapeType=Capsule
   length=12.6372
   radius=3.1231
   shapeName=dyn_rope_04_r_capsule
   center=<8.9681 0.2863 -0.0566>
   rotation=<rot 0 -87.6927 -89.7994>
  }]
 dyn_top_rope_l [{shapeType=Capsule, length=2, radius=2, shapeName=dyn_top_rope_l_capsule, center=<2.5 0 0>}]
 dyn_top_rope_r [{shapeType=Capsule, length=2, radius=2, shapeName=dyn_top_rope_r_capsule, center=<2.5 0 0>}]
 dyn_zipper [{shapeType=Capsule, length=0, radius=0.5, shapeName=base_zipper_capsule, rotation=<rot 0 -90 0>}]
 pack [
  {shapeType=Capsule, length=3.5839, radius=9.5, center=<0.1318 -3.5039 -6.719>, rotation=<rot 2.3551 -175.542 -21.8159>}
  {shapeType=Capsule, length=32, radius=16, shapeName=pack_capsule, center=<0 13.5595 -14.1192>}
  {shapeType=Capsule, length=8.5, radius=13, shapeName=pack_capsule, center=<0 10.2704 -38.603>, rotation=<rot 90 -4.0447 -4.0447>}
 ]
}
constraints [
 {
  bone1Name=base_rope_l
  bone2Name=pack
  swing1LimitDegrees=45
  swing1Motion=Free
  swing2LimitDegrees=45
  swing2Motion=Free
  twistLimitDegrees=45
  twistMotion=Free
 }
 {bone1Name=dyn_rope_00_l, bone2Name=base_rope_l, swing1LimitDegrees=15, swing2LimitDegrees=15, twistLimitDegrees=45, twistMotion=Locked}
 {bone1Name=dyn_rope_01_l, bone2Name=dyn_rope_00_l, swing1LimitDegrees=30, swing2LimitDegrees=0, twistLimitDegrees=45, twistMotion=Locked}
 {bone1Name=dyn_rope_02_l, bone2Name=dyn_rope_01_l, swing1LimitDegrees=60, swing2LimitDegrees=60, twistLimitDegrees=45, twistMotion=Locked}
 {bone1Name=dyn_rope_03_l, bone2Name=dyn_rope_02_l, swing1LimitDegrees=40, swing2LimitDegrees=40, twistLimitDegrees=45, twistMotion=Locked}
 {bone1Name=dyn_rope_04_l, bone2Name=dyn_rope_03_l, swing1LimitDegrees=50, swing2LimitDegrees=50, twistLimitDegrees=45, twistMotion=Locked}
 {
  bone1Name=base_rope_r
  bone2Name=pack
  swing1LimitDegrees=45
  swing1Motion=Free
  swing2LimitDegrees=45
  swing2Motion=Free
  twistLimitDegrees=45
  twistMotion=Free
 }
 {bone1Name=dyn_rope_00_r, bone2Name=base_rope_r, swing1LimitDegrees=15, swing2LimitDegrees=15, twistLimitDegrees=45, twistMotion=Locked}
 {bone1Name=dyn_rope_01_r, bone2Name=dyn_rope_00_r, swing1LimitDegrees=30, swing2LimitDegrees=0, twistLimitDegrees=45, twistMotion=Locked}
 {bone1Name=dyn_rope_02_r, bone2Name=dyn_rope_01_r, swing1LimitDegrees=45, swing2LimitDegrees=45, twistLimitDegrees=45, twistMotion=Locked}
 {bone1Name=dyn_rope_03_r, bone2Name=dyn_rope_02_r, swing1LimitDegrees=30, swing2LimitDegrees=30, twistLimitDegrees=45, twistMotion=Locked}
 {bone1Name=dyn_rope_04_r, bone2Name=dyn_rope_03_r, swing1LimitDegrees=40, swing2LimitDegrees=40, twistLimitDegrees=45, twistMotion=Locked}
 {
  bone1Name=base_top_rope_l
  bone2Name=pack
  swing1LimitDegrees=45
  swing1Motion=Free
  swing2LimitDegrees=45
  swing2Motion=Free
  twistLimitDegrees=45
  twistMotion=Free
 }
 {bone1Name=dyn_top_rope_l, bone2Name=base_top_rope_l, swing1LimitDegrees=15, swing2LimitDegrees=10, twistLimitDegrees=10}
 {
  bone1Name=base_top_rope_r
  bone2Name=pack
  swing1LimitDegrees=45
  swing1Motion=Free
  swing2LimitDegrees=45
  swing2Motion=Free
  twistLimitDegrees=45
  twistMotion=Free
 }
 {bone1Name=dyn_top_rope_r, bone2Name=base_top_rope_r, swing1LimitDegrees=15, swing2LimitDegrees=10, twistLimitDegrees=10}
 {
  bone1Name=base_zipper
  bone2Name=pack
  swing1LimitDegrees=45
  swing1Motion=Free
  swing2LimitDegrees=45
  swing2Motion=Free
  twistLimitDegrees=45
  twistMotion=Free
 }
 {bone1Name=dyn_zipper, bone2Name=base_zipper, swing1LimitDegrees=90, swing2LimitDegrees=10, twistLimitDegrees=45, twistMotion=Free}
]
