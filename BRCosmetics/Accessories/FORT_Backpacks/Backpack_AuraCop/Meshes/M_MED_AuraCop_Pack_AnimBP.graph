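# Animation Blueprint for the AuraCop backpack, adding secondary motion and pose corrections.
# Body type selects a Large or standard RigidBody physics branch, both using gravity overrides and crouch rope-bone offsets.
# The standard branch adds AnimDynamics zipper motion, a hang-glide glute_push offset and crouch-moving rope rotations.
# The event graph caches pawn and movement state and toggles a material element based on AuraCop-tagged pickaxe and skydiving.
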AnimBlueprint /BRCosmetics/Accessories/FORT_Backpacks/Backpack_AuraCop/Meshes/M_MED_AuraCop_Pack_AnimBP
target_skeleton=&/M_MED_AuraCop_Pack_Skeleton
graphs [
 AnimGraph {nodes [
   n1 = Get Hang Gliding Float() -> (Hang Gliding Float:double)
   n2 = "Transform (Modify) Bone - Bone: base_rope_l"(Alpha=@n1, Translation="0,1,0", Rotation="0,-20,0", Scale="1,1,1") {
    rotationMode=BMM_Additive
    rotationSpace=BCS_BoneSpace
    translationMode=BMM_Additive
    translationSpace=BCS_BoneSpace
    alphaBoolBlend {blendOption=Linear}
    boneToModify {boneName=base_rope_l}
    rotation {yaw=-20}
    translation {y=1}
   }
   "Transform (Modify) Bone - Bone: base_rope_r"(ComponentPose=@n2, Alpha=@n1, Translation="0,1,0", Rotation="0,-20,0", Scale="1,1,1") {
    rotationMode=BMM_Additive
    rotationSpace=BCS_BoneSpace
    translationMode=BMM_Additive
    translationSpace=BCS_BoneSpace
    alphaBoolBlend {blendOption=Linear}
    boneToModify {boneName=base_rope_r}
    rotation {yaw=-20}
    translation {y=1}
   }
   "Transform (Modify) Bone - Bone: glute_push"!(ComponentPose, bAlphaBoolEnabled:bool=Get bIsSkyDiving(), Translation="0,-20,10", Rotation="0,0,0", Scale="1,1,1") -> (Pose) {
    alphaInputType=Bool
    rotationMode=BMM_Additive
    rotationSpace=BCS_BoneSpace
    translationMode=BMM_Additive
    translationSpace=BCS_BoneSpace
    alphaBoolBlend {blendOption=Linear}
    boneToModify {boneName=glute_push}
    translation {y=-20, z=10}
   }
   n3 = RigidBody!(
    ComponentPose=Local To Component(LocalPose=Use cached pose 'Initial Pose'())
    Alpha
    OverrideWorldGravity:Vector=Get GravityOverride()
    ExternalForce
   ) -> (Pose) {
    bClampLinearTranslationLimitToRefPose=true
    bOverrideWorldGravity=true
    overridePhysicsAsset=&/M_MED_AuraCop_Pack_PhysicsAsset_Large
    simulationSpace=BaseBoneSpace
    alphaBoolBlend {blendOption=Linear}
    baseBoneRef {boneName=root}
    componentAppliedLinearAccClamp=<3000 3000 3000>
    componentLinearAccScale=<0.4 0.3 0.3>
    componentLinearVelScale=<0.4 0.3 0.3>
    overrideWorldGravity {z=-980}
    simSpaceSettings {
     maxAngularAcceleration=1000
     maxAngularVelocity=3
     maxLinearAcceleration=700
     maxLinearVelocity=500
     velocityScaleZ=0
     worldAlpha=0.1
     externalLinearDragV=<0.2 0.2>
    }
   }
   n4 = Get bIsCrouching()
   n5 = "Transform (Modify) Bone - Bone: base_rope_l"!(ComponentPose=@n3, bAlphaBoolEnabled:bool=@n4, Translation="0,-0.5,0", Rotation="0,0,0", Scale="1,1,1") -> (Pose) {
    alphaInputType=Bool
    rotationMode=BMM_Additive
    rotationSpace=BCS_BoneSpace
    translationMode=BMM_Additive
    translationSpace=BCS_BoneSpace
    alphaBoolBlend {blendOption=Linear}
    boneToModify {boneName=base_rope_l}
    translation {y=-0.5}
   }
   n6 = "Transform (Modify) Bone - Bone: base_rope_r"!(ComponentPose=@n5, bAlphaBoolEnabled:bool=@n4, Translation="0,-0.5,0", Rotation="0,0,0", Scale="1,1,1") -> (Pose) {
    alphaInputType=Bool
    rotationMode=BMM_Additive
    rotationSpace=BCS_BoneSpace
    translationMode=BMM_Additive
    translationSpace=BCS_BoneSpace
    alphaBoolBlend {blendOption=Linear}
    boneToModify {boneName=base_rope_r}
    translation {y=-0.5}
   }
   n7 = "Transform (Modify) Bone - Bone: glute_push"!(
    ComponentPose=Local To Component(LocalPose=Use cached pose 'Initial Pose'())
    Alpha:float=Get Hang Gliding Float()
    Translation="0,-20,10"
    Rotation="0,0,0"
    Scale="1,1,1"
   ) -> (Pose) {
    rotationMode=BMM_Additive
    rotationSpace=BCS_BoneSpace
    translationMode=BMM_Additive
    translationSpace=BCS_BoneSpace
    alphaBoolBlend {blendOption=Linear}
    boneToModify {boneName=glute_push}
    translation {y=-20, z=10}
   }
   n8 = Anim Dynamics - Bone: dyn_zipper(ComponentPose=@n7, Alpha=1) {
    angularDampingOverride=1
    angularSpringConstant=10000
    bAngularSpring=true
    bOverrideAngularDamping=true
    bUseGravityOverride=true
    gravityScale=0.35
    simulationSpace=RootRelative
    alphaBoolBlend {blendOption=Linear}
    boundBone {boneName=dyn_zipper}
    componentLinearAccScale=<1 1 1>
    componentLinearVelScale=<1 1 1>
    gravityOverride {z=-980}
    physicsBodyDefinitions [{
      boundBone {boneName=dyn_zipper}
      boxExtents=<6 6 6>
      constraintSetup {angularLimitsMax=<1 1 0>, angularLimitsMin=<-1 -1 -15>, angularTarget=<1 0 0>}
      localJointOffset=<4 0 0>
     }]
   }
   n9 = RigidBody(ComponentPose=@n8, OverrideWorldGravity=Get GravityOverride()) {
    bClampLinearTranslationLimitToRefPose=true
    bOverrideWorldGravity=true
    overridePhysicsAsset=&/M_MED_AuraCop_Pack_PhysicsAsset
    simulationSpace=BaseBoneSpace
    alphaBoolBlend {blendOption=Linear}
    baseBoneRef {boneName=root}
    componentAppliedLinearAccClamp=<3000 3000 3000>
    componentLinearAccScale=<0.4 0.3 0.3>
    componentLinearVelScale=<0.4 0.3 0.3>
    overrideWorldGravity {z=-980}
    simSpaceSettings {
     maxAngularAcceleration=1000
     maxAngularVelocity=3
     maxLinearAcceleration=700
     maxLinearVelocity=500
     velocityScaleZ=0
     worldAlpha=0.1
     externalLinearDragV=<0.2 0.2>
    }
   }
   n10 = Get bIsCrouching()
   n11 = "Transform (Modify) Bone - Bone: base_rope_l"(ComponentPose=@n9, bAlphaBoolEnabled=@n10, Translation="0,-1,0", Rotation="0,0,0", Scale="1,1,1") {
    alphaInputType=Bool
    rotationMode=BMM_Additive
    rotationSpace=BCS_BoneSpace
    translationMode=BMM_Additive
    translationSpace=BCS_BoneSpace
    alphaBoolBlend {blendOption=Linear}
    boneToModify {boneName=base_rope_l}
    translation {y=-1}
   }
   n12 = "Transform (Modify) Bone - Bone: base_rope_r"(ComponentPose=@n11, bAlphaBoolEnabled=@n10, Translation="0,-1,0", Rotation="0,0,0", Scale="1,1,1") {
    alphaInputType=Bool
    rotationMode=BMM_Additive
    rotationSpace=BCS_BoneSpace
    translationMode=BMM_Additive
    translationSpace=BCS_BoneSpace
    alphaBoolBlend {blendOption=Linear}
    boneToModify {boneName=base_rope_r}
    translation {y=-1}
   }
   n13 = Get Crouch moving alpha() -> (Crouch moving alpha:double)
   n14 = "Transform (Modify) Bone - Bone: base_rope_l"!(ComponentPose=@n12, Alpha:float=@n13, Translation="0,0,0", Rotation="-10,-10,0", Scale="1,1,1") -> (Pose) {
    rotationMode=BMM_Additive
    rotationSpace=BCS_BoneSpace
    translationMode=BMM_Additive
    translationSpace=BCS_BoneSpace
    alphaBoolBlend {blendOption=Linear}
    alphaScaleBiasClamp {bInterpResult=true}
    boneToModify {boneName=base_rope_l}
    rotation {pitch=-10, yaw=-10}
   }
   n15 = "Transform (Modify) Bone - Bone: base_rope_r"!(ComponentPose=@n14, Alpha:float=@n13, Translation="0,0,0", Rotation="-10,-10,0", Scale="1,1,1") -> (Pose) {
    rotationMode=BMM_Additive
    rotationSpace=BCS_BoneSpace
    translationMode=BMM_Additive
    translationSpace=BCS_BoneSpace
    alphaBoolBlend {blendOption=Linear}
    alphaScaleBiasClamp {bInterpResult=true}
    boneToModify {boneName=base_rope_r}
    rotation {pitch=-10, yaw=-10}
   }
   Output Pose(Result=Blend Poses by bool(
     BlendPose_0=Component To Local(ComponentPose=@n6)
     BlendPose_1=Component To Local(ComponentPose=@n15)
     bActiveValue=Get IsLarge() -> (IsLarge:bool)
    ) {blendType=Linear})
   n16 = "Transform (Modify) Bone - Bone: pack"(ComponentPose=Local To Component(LocalPose=Input Pose()), Translation="0,0.5,0", Rotation="0,0,0") {
    translationMode=BMM_Additive
    translationSpace=BCS_BoneSpace
    alphaBoolBlend {blendOption=Linear}
    boneToModify {boneName=pack}
    translation {y=0.5}
   }
   Save cached pose 'Initial Pose'(Pose=Component To Local(ComponentPose=@n16)) {cacheName=Initial Pose}
  ]}
 EventGraph {nodes [
   Delay(Duration=0.2)
   > ResetDynamics(InTeleportType=TeleportPhysics)
   > ResetDynamics(InTeleportType=ResetPhysics)
   Event BlueprintUpdateAnimation()
   > Sequence()
   > n1 = Gravity Override State Function DEPRECATED(CharacterPartAnimInstance=Self-Reference(), GameTimeInSeconds=GetGameTimeInSeconds(), GravityOverrideParams=Get Gravity Override Params())
   > Set GravityOverride(GravityOverride=@n1)
   > n2 = Set Crouch moving(Crouch moving:bool=AND Boolean(A=Get bIsMoving2D(), B=Get bIsCrouching())) -> (Output_Get:bool)
   > Set Crouch moving alpha(Crouch moving alpha:double="To Float (Boolean)"(InBool=@n2)) -> (Output_Get:double)
   Event OnEndSkydiving()
   n3 = PickaxeItemDefinitionHasAnyMetaTag(Tags=Make Array("[0]"=(TagName="Cosmetics.MetaTags.AuraCop")), bRequiresEquipped=true)
   > n4 = Branch(Condition=@n3)
   n5 = Get Dynamic Material Instance(execute=[@n6.then, @n8.then_1], SkeletalMeshComponent=GetPartSkeletalMeshComponent())
   SetScalarParameterValue(execute=[@n4.then, @n16.else], self=@n5, ParameterName=Hide Element 01, Value=1)
   SetScalarParameterValue(execute=[@n7.then, @n4.else, @n16.then, @n15.else], self=@n5, ParameterName=Hide Element 01, Value=0)
   n6 = Event OnWeaponChanged()
   n7 = Event OnBeginSkydiving()
   Event OnParachuteOpened()
   > Switch on EGliderType(Selection=Get PlayerGliderType() -> (PlayerGliderType=HangGlider))
   > Set Hang Gliding(Hang Gliding:bool=true) -> (Output_Get:bool)
   > Set Hang Gliding Float(Hang Gliding Float:double=1) -> (Output_Get:double)
   Event OnCharacterCustomizationLoaded()
   > n8 = Sequence!(execute) -> (then_0, then_1, then_2:Exec)
   n9 = Set Owning Pawn(execute=@n8.then_2, Owning Pawn:Pawn Object Reference=TryGetPawnOwner()) -> (Output_Get:Pawn Object Reference)
   > Is Valid(InputObject=@n9)
   > n10 = Sequence()
   n11 = FindBodyType!(execute=@n10.then_1, Object:Object Reference=Get Owning Pawn() -> (Owning Pawn:Pawn Object Reference)) -> (True:Exec, False:Exec, OutBodyType:EFortCustomBodyType Enum=NONE)
   > Set IsLarge(IsLarge:bool="Equal (Enum)"(A:EFortCustomBodyType Enum=@n11, B:EFortCustomBodyType Enum=Large)) -> (Output_Get:bool)
   n12 = Cast To FortPlayerPawn(execute=@n10.then_0, Object=Get Owning Pawn())
   > n13 = Set Owning Player Pawn(Owning Player Pawn:Fort Player Pawn Object Reference=@n12) -> (Output_Get:Fort Player Pawn Object Reference)
   > Set BodyMesh(BodyMesh=GetSkeletalMeshForPartType(self=@n13))
   n14 = FindGender!(execute=@n10.then_2, Object:Object Reference=Get Owning Pawn()) -> (True:Exec, False:Exec, OutGender:EFortCustomGender Enum=Invalid)
   > Set isFemale(isFemale:bool="Equal (Enum)"(A:EFortCustomGender Enum=@n14, B:EFortCustomGender Enum=Female)) -> (Output_Get:bool)
   Event OnParachuteClosed()
   > Set Hang Gliding(Hang Gliding=false)
   > Set Hang Gliding Float(Hang Gliding Float=0)
   TryGetPawnOwner()
   n15 = Branch(execute=@n5.then, Condition=@n3)
   > n16 = Branch(Condition=Get bIsSkyDiving())
  ]}
]
variables [
 BodyMesh {type=SkeletalMeshComponent}
 Crouch moving {type=boolean}
 Crouch moving alpha {type=number}
 Gravity Override Params {type=GravityOverrideParamsStruct}
 GravityOverride {type=Vector}
 Hang Gliding {type=boolean}
 Hang Gliding Float {type=number}
 isFemale {type=boolean}
 IsLarge {type=boolean}
 Owning Pawn {type=Pawn, default=null}
 Owning Player Pawn {type=FortPlayerPawn}
]
